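# Animation Blueprint for the M_MED_MicroLeaf player body cosmetic, adding secondary motion to its tail.
# The parent mesh pose passes through a Control Rig into two spring-based Anim Dynamics chains on tail1–4, differing mainly in stiffness and selected by front-end cape idle.
# Stacked additive bone offsets then reposition tail1–3.
# The event graph derives cape, movement, specific-emote and hang-glider states, plus a gravity override from a parameter asset.
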AnimBlueprint /BRCosmetics/Characters/Player/Male/Medium/Bodies/M_MED_MicroLeaf/Meshes/M_MED_MicroLeaf_AnimBP
target_skeleton=&/M_MED_MicroLeaf_Skeleton
graphs [
 AnimGraph {nodes [
   n1 = "Transform (Modify) Bone - Bone: tail1"!(
    ComponentPose=Local To Component(LocalPose=Use cached pose 'Dynamics Cache'())
    bAlphaBoolEnabled:bool=True
    Translation="0.169,-2.214,-0.724"
    Rotation="49.276,-0.003,-0.002"
    Scale="1,1,1"
   ) -> (Pose) {
    alphaInputType=Bool
    rotationMode=BMM_Additive
    rotationSpace=BCS_BoneSpace
    translationMode=BMM_Additive
    alphaBoolBlend {blendInTime=0.8, blendOutTime=0.2}
    boneToModify {boneName=tail1}
    rotation=<rot 49.276 -0.003 -0.002>
    translation=<0.169 -2.2142 -0.7241>
   }
   n2 = "Transform (Modify) Bone - Bone: tail1"(ComponentPose=@n1, bAlphaBoolEnabled=True, Translation="0.116,-4.557,2.668", Rotation="69.940,0.989,6.771", Scale="1,1,1") {
    alphaInputType=Bool
    rotationMode=BMM_Additive
    rotationSpace=BCS_BoneSpace
    translationMode=BMM_Additive
    alphaBoolBlend {blendInTime=0.8, blendOutTime=0.2}
    boneToModify {boneName=tail1}
    rotation=<rot 69.94 0.989 6.771>
    translation=<0.1158 -4.5567 2.668>
   }
   n3 = "Transform (Modify) Bone - Bone: tail1"(ComponentPose=@n2, bAlphaBoolEnabled=True, Translation="0,0,0", Rotation="-20,0,0", Scale="1,1,1") {
    alphaInputType=Bool
    rotationMode=BMM_Additive
    rotationSpace=BCS_BoneSpace
    translationMode=BMM_Additive
    alphaBoolBlend {blendInTime=0.8, blendOutTime=0.2}
    boneToModify {boneName=tail1}
    rotation {pitch=-20}
   }
   n4 = "Transform (Modify) Bone - Bone: tail1"!(ComponentPose=@n3, Alpha:float=1, Translation="-0.782,2.170,-1.206", Rotation="-43.378,-0.022,0.014", Scale="1,1,1") -> (Pose) {
    rotationMode=BMM_Additive
    rotationSpace=BCS_BoneSpace
    translationMode=BMM_Additive
    alphaBoolBlend {blendInTime=0.8, blendOutTime=0.2}
    alphaScaleBiasClamp {bInterpResult=true, interpSpeedDecreasing=5}
    boneToModify {boneName=tail1}
    rotation=<rot -43.378 -0.022 0.014>
    translation=<-0.782 2.17 -1.206>
   }
   n5 = "Transform (Modify) Bone - Bone: tail2"(ComponentPose=@n4, Alpha=1, Translation="-0.945,3.517,-0.858", Rotation="-36.971,0.005,-0.004", Scale="1,1,1") {
    rotationMode=BMM_Additive
    rotationSpace=BCS_BoneSpace
    translationMode=BMM_Additive
    alphaBoolBlend {blendInTime=0.8, blendOutTime=0.2}
    alphaScaleBiasClamp {bInterpResult=true, interpSpeedDecreasing=5}
    boneToModify {boneName=tail2}
    rotation=<rot -36.971 0.005 -0.004>
    translation=<-0.945 3.5168 -0.8579>
   }
   n6 = "Transform (Modify) Bone - Bone: tail3"(ComponentPose=@n5, Alpha=1, Translation="0,0,0", Rotation="-23.526,0.002,0.002", Scale="1,1,1") {
    rotationMode=BMM_Additive
    rotationSpace=BCS_BoneSpace
    translationMode=BMM_Additive
    alphaBoolBlend {blendInTime=0.8, blendOutTime=0.2}
    alphaScaleBiasClamp {bInterpResult=true, interpSpeedDecreasing=5}
    boneToModify {boneName=tail3}
    rotation=<rot -23.526 0.002 0.002>
   }
   n7 = "Transform (Modify) Bone - Bone: tail2"(ComponentPose=@n6, Alpha=1, Translation="0.001,-3.393,-2.532", Rotation="-11.632,0.023,-0.028", Scale="1,1,1") {
    rotationMode=BMM_Additive
    rotationSpace=BCS_BoneSpace
    translationMode=BMM_Additive
    alphaBoolBlend {blendInTime=0.8, blendOutTime=0.2}
    boneToModify {boneName=tail2}
    rotation=<rot -11.632 0.023 -0.028>
    translation=<0.001 -3.3928 -2.5319>
   }
   n8 = "Transform (Modify) Bone - Bone: tail3"(ComponentPose=@n7, Alpha=1, Translation="0,0,0", Rotation="-22.564,-0.000,-0.000", Scale="1,1,1") {
    rotationMode=BMM_Additive
    rotationSpace=BCS_BoneSpace
    translationMode=BMM_Additive
    alphaBoolBlend {blendInTime=0.8, blendOutTime=0.2}
    boneToModify {boneName=tail3}
    rotation {pitch=-22.564, roll=0}
   }
   n9 = "Transform (Modify) Bone - Bone: tail1"(ComponentPose=@n8, Alpha=1, Translation="0,0,0", Rotation="-35,0,0", Scale="1,1,1") {
    rotationMode=BMM_Additive
    rotationSpace=BCS_BoneSpace
    translationMode=BMM_Additive
    alphaBoolBlend {blendInTime=0.2, blendOutTime=0.2}
    alphaScaleBiasClamp {bInterpResult=true}
    boneToModify {boneName=tail1}
    rotation {pitch=-35}
   }
   n10 = "Transform (Modify) Bone - Bone: tail2"(ComponentPose=@n9, Alpha=1, Translation="0,0,0", Rotation="-5,0,0", Scale="1,1,1") {
    rotationMode=BMM_Additive
    rotationSpace=BCS_BoneSpace
    translationMode=BMM_Additive
    alphaBoolBlend {blendInTime=0.2, blendOutTime=0.2}
    alphaScaleBiasClamp {bInterpResult=true}
    boneToModify {boneName=tail2}
    rotation {pitch=-5}
   }
   n11 = "Transform (Modify) Bone - Bone: tail1"(ComponentPose=@n10, Alpha=1, Translation="0,0,0", Rotation="20,0,0", Scale="1,1,1") {
    rotationMode=BMM_Additive
    rotationSpace=BCS_BoneSpace
    translationMode=BMM_Additive
    alphaBoolBlend {blendInTime=0.2, blendOutTime=0.2}
    alphaScaleBiasClamp {bInterpResult=true, interpSpeedIncreasing=7}
    boneToModify {boneName=tail1}
    rotation {pitch=20}
   }
   n12 = "Transform (Modify) Bone - Bone: tail1"(ComponentPose=@n11, Alpha=1, Translation="0,0,0", Rotation="12.5,0,0", Scale="1,1,1") {
    rotationMode=BMM_Additive
    rotationSpace=BCS_BoneSpace
    translationMode=BMM_Additive
    alphaBoolBlend {blendInTime=0.2, blendOutTime=0.2}
    alphaScaleBiasClamp {bInterpResult=true}
    boneToModify {boneName=tail1}
    rotation {pitch=12.5}
   }
   Output Pose(Result=Component To Local(ComponentPose=@n12)) {layerGroup=None}
   n13 = Anim Dynamics - Bone: tail1!(
    ComponentPose=Local To Component(LocalPose=Use cached pose 'CtrlRigCache'())
    Alpha=1
    GravityOverride:Vector="0,0,-1500"
    ExternalForce="0,0,0"
   ) -> (Pose) {
    angularDampingOverride=1
    angularSpringConstant=5000
    bAngularSpring=true
    bOverrideAngularDamping=true
    bUseGravityOverride=true
    gravityScale=0.35
    simulationSpace=RootRelative
    alphaBoolBlend {blendOption=Linear}
    boundBone {boneName=tail1}
    componentAppliedLinearAccClamp=<3000 3000 3000>
    componentLinearAccScale=<1 1 1>
    componentLinearVelScale=<1 1 1>
    gravityOverride {z=-1500}
    physicsBodyDefinitions [{
      boundBone {boneName=tail1}
      boxExtents=<23 23 23>
      constraintSetup {angularLimitsMax=<0 5 30>, angularLimitsMin=<-30 -30 -30>, angularTarget=<1 0 0>}
      localJointOffset=<1 0 0>
     }]
    simSpaceSettings {maxAngularAcceleration=100, maxAngularVelocity=100, simSpaceAngularAlpha=0.5}
   }
   n14 = Anim Dynamics - Bone: tail2!(ComponentPose=@n13, Alpha=1, GravityOverride:Vector="0,0,-1500", ExternalForce="0,0,0") -> (Pose) {
    angularDampingOverride=1
    angularSpringConstant=5000
    bAngularSpring=true
    bOverrideAngularDamping=true
    bUseGravityOverride=true
    gravityScale=0.35
    simulationSpace=RootRelative
    alphaBoolBlend {blendOption=Linear}
    boundBone {boneName=tail2}
    componentAppliedLinearAccClamp=<3000 3000 3000>
    componentLinearAccScale=<1 1 1>
    componentLinearVelScale=<1 1 1>
    gravityOverride {z=-1500}
    physicsBodyDefinitions [{
      boundBone {boneName=tail2}
      boxExtents=<23 23 23>
      constraintSetup {angularLimitsMax=<0 5 30>, angularLimitsMin=<-30 -30 -30>, angularTarget=<1 0 0>}
      localJointOffset=<1 0 0>
     }]
    simSpaceSettings {maxAngularAcceleration=100, maxAngularVelocity=100, simSpaceAngularAlpha=0.5}
   }
   n15 = Anim Dynamics - Bone: tail3!(ComponentPose=@n14, Alpha=1, GravityOverride:Vector="0,0,-1500", ExternalForce="0,0,0") -> (Pose) {
    angularDampingOverride=1
    angularSpringConstant=5000
    bAngularSpring=true
    bOverrideAngularDamping=true
    bUseGravityOverride=true
    gravityScale=0.35
    simulationSpace=RootRelative
    alphaBoolBlend {blendOption=Linear}
    boundBone {boneName=tail3}
    componentAppliedLinearAccClamp=<3000 3000 3000>
    componentLinearAccScale=<1 1 1>
    componentLinearVelScale=<1 1 1>
    gravityOverride {z=-1500}
    physicsBodyDefinitions [{
      boundBone {boneName=tail3}
      boxExtents=<23 23 23>
      constraintSetup {angularLimitsMax=<0 5 30>, angularLimitsMin=<-30 -30 -30>, angularTarget=<1 0 0>}
      localJointOffset=<1 0 0>
     }]
    simSpaceSettings {maxAngularAcceleration=100, maxAngularVelocity=100, simSpaceAngularAlpha=0.5}
   }
   n16 = Anim Dynamics - Bone: tail4!(ComponentPose=@n15, Alpha=1, GravityOverride:Vector="0,0,-1500", ExternalForce="0,0,0") -> (Pose) {
    angularDampingOverride=1
    angularSpringConstant=5000
    bAngularSpring=true
    bOverrideAngularDamping=true
    bUseGravityOverride=true
    gravityScale=0.35
    simulationSpace=RootRelative
    alphaBoolBlend {blendOption=Linear}
    boundBone {boneName=tail4}
    componentAppliedLinearAccClamp=<3000 3000 3000>
    componentLinearAccScale=<1 1 1>
    componentLinearVelScale=<1 1 1>
    gravityOverride {z=-1500}
    physicsBodyDefinitions [{
      boundBone {boneName=tail4}
      boxExtents=<23 23 23>
      constraintSetup {angularLimitsMax=<0 5 30>, angularLimitsMin=<-30 -30 -30>, angularTarget=<1 0 0>}
      localJointOffset=<1 0 0>
     }]
    simSpaceSettings {maxAngularAcceleration=100, maxAngularVelocity=100, simSpaceAngularAlpha=0.5}
   }
   n17 = Anim Dynamics - Bone: tail1(ComponentPose=Local To Component(LocalPose=Use cached pose 'CtrlRigCache'()), Alpha=1, GravityOverride="0,0,-1500", ExternalForce="0,0,0") {
    angularDampingOverride=1
    angularSpringConstant=20000
    bAngularSpring=true
    bOverrideAngularDamping=true
    bUseGravityOverride=true
    gravityScale=0.35
    simulationSpace=RootRelative
    alphaBoolBlend {blendOption=Linear}
    boundBone {boneName=tail1}
    componentAppliedLinearAccClamp=<3000 3000 3000>
    componentLinearAccScale=<1 1 1>
    componentLinearVelScale=<1 1 1>
    gravityOverride {z=-1500}
    physicsBodyDefinitions [{
      boundBone {boneName=tail1}
      boxExtents=<23 23 23>
      constraintSetup {angularLimitsMax=<0 0 30>, angularLimitsMin=<-30 -30 -30>, angularTarget=<1 0 0>}
      localJointOffset=<1 0 0>
     }]
    simSpaceSettings {maxAngularAcceleration=100, maxAngularVelocity=100, simSpaceAngularAlpha=0.5}
   }
   n18 = Anim Dynamics - Bone: tail2(ComponentPose=@n17, Alpha=1, GravityOverride="0,0,-1500", ExternalForce="0,0,0") {
    angularDampingOverride=1
    angularSpringConstant=20000
    bAngularSpring=true
    bOverrideAngularDamping=true
    bUseGravityOverride=true
    gravityScale=0.35
    simulationSpace=RootRelative
    alphaBoolBlend {blendOption=Linear}
    boundBone {boneName=tail2}
    componentAppliedLinearAccClamp=<3000 3000 3000>
    componentLinearAccScale=<1 1 1>
    componentLinearVelScale=<1 1 1>
    gravityOverride {z=-1500}
    physicsBodyDefinitions [{
      boundBone {boneName=tail2}
      boxExtents=<23 23 23>
      constraintSetup {angularLimitsMax=<0 0 30>, angularLimitsMin=<-30 -30 -30>, angularTarget=<1 0 0>}
      localJointOffset=<1 0 0>
     }]
    simSpaceSettings {maxAngularAcceleration=100, maxAngularVelocity=100, simSpaceAngularAlpha=0.5}
   }
   n19 = Anim Dynamics - Bone: tail3(ComponentPose=@n18, Alpha=1, GravityOverride="0,0,-1500", ExternalForce="0,0,0") {
    angularDampingOverride=1
    angularSpringConstant=20000
    bAngularSpring=true
    bOverrideAngularDamping=true
    bUseGravityOverride=true
    gravityScale=0.35
    simulationSpace=RootRelative
    alphaBoolBlend {blendOption=Linear}
    boundBone {boneName=tail3}
    componentAppliedLinearAccClamp=<3000 3000 3000>
    componentLinearAccScale=<1 1 1>
    componentLinearVelScale=<1 1 1>
    gravityOverride {z=-1500}
    physicsBodyDefinitions [{
      boundBone {boneName=tail3}
      boxExtents=<23 23 23>
      constraintSetup {angularLimitsMax=<0 0 30>, angularLimitsMin=<-30 -30 -30>, angularTarget=<1 0 0>}
      localJointOffset=<1 0 0>
     }]
    simSpaceSettings {maxAngularAcceleration=100, maxAngularVelocity=100, simSpaceAngularAlpha=0.5}
   }
   n20 = Anim Dynamics - Bone: tail4(ComponentPose=@n19, Alpha=1, GravityOverride="0,0,-1500", ExternalForce="0,0,0") {
    angularDampingOverride=1
    angularSpringConstant=20000
    bAngularSpring=true
    bOverrideAngularDamping=true
    bUseGravityOverride=true
    gravityScale=0.35
    simulationSpace=RootRelative
    alphaBoolBlend {blendOption=Linear}
    boundBone {boneName=tail4}
    componentAppliedLinearAccClamp=<3000 3000 3000>
    componentLinearAccScale=<1 1 1>
    componentLinearVelScale=<1 1 1>
    gravityOverride {z=-1500}
    physicsBodyDefinitions [{
      boundBone {boneName=tail4}
      boxExtents=<23 23 23>
      constraintSetup {angularLimitsMax=<0 0 30>, angularLimitsMin=<-30 -30 -30>, angularTarget=<1 0 0>}
      localJointOffset=<1 0 0>
     }]
    simSpaceSettings {maxAngularAcceleration=100, maxAngularVelocity=100, simSpaceAngularAlpha=0.5}
   }
   Save cached pose 'Dynamics Cache'(Pose=Blend Poses by bool(
     BlendPose_0=Component To Local(ComponentPose=@n20)
     BlendPose_1=Component To Local(ComponentPose=@n16)
     bActiveValue=Get CapeFEIdle_Bool() -> (CapeFEIdle_Bool:bool)
    )) {cacheName=Dynamics Cache}
   n21 = Control Rig!(
    Source=Copy Pose From Mesh() {bUseAttachedParent=true}
    Alpha
    HangGlidingValue:float=Get HangGlidingValue() -> (HangGlidingValue:double)
    HangGliding:bool=Get HangGliding() -> (HangGliding:bool)
    AdjustTailToEmote:float=Get AdjustTailToEmotes() -> (AdjustTailToEmotes:double)
    HandTwistEnable:double=Get HandTwistEnable() -> (HandTwistEnable:double)
   ) -> (Pose) {controlRigAssetReference {blueprintRigClass=&/M_MED_MicroLeaf_CtrlRig.M_MED_MicroLeaf_CtrlRig_C}}
   Save cached pose 'CtrlRigCache'(Pose=@n21) {cacheName=CtrlRigCache}
  ]}
 EventGraph {nodes [
   Event BlueprintUpdateAnimation()
   > n1 = Sequence!(execute) -> (then_0, then_1, then_2:Exec)
   > n2 = "Gravity Override State Function (Parameter Asset) (Nativized)"(
    CharacterPartAnimInstance=Self-Reference()
    GameTimeInSeconds=GetGameTimeInSeconds()
    DebugJointName=spine_05
    Parameters=&/M_MED_MicroLeafParemeters
   )
   > Set Gravity Override(Gravity Override=@n2)
   n3 = GetAnimInstance(self=GetOwnerSkeletalMeshComponent())
   n4 = OR Boolean(
    A=Montage_IsPlaying(self=@n3, Montage=/BRCosmetics/Animation/Game/MainPlayer/Emotes/Cauldron/Emote_Cauldron_CMM_M)
    B=Montage_IsPlaying(self=@n3, Montage=/BRCosmetics/Animation/Game/MainPlayer/Emotes/Scrub_A_Dub/Emote_ScrubaDub_CMM_M)
   )
   Set AdjustTailToEmotes(execute=@n1.then_1, AdjustTailToEmotes:double="To Float (Boolean)"(InBool=@n4)) -> (Output_Get:double)
   > Set AdjustTailToEmotesBool(AdjustTailToEmotesBool:bool=@n4) -> (Output_Get:bool)
   n5 = CheckForBackbling()
   Set Is Wearing Cape(execute=@n1.then_2, Is Wearing Cape=OR Boolean(A=@n5.bIsBackblingACape, B=@n5.bIsBackblingLong))
   n6 = AND Boolean!(
    A=Get bIsInFrontEnd()
    B=Get Is Wearing Cape() -> (Is Wearing Cape:bool)
    C:bool=NOT Boolean(A=Get bIsMoving2D())
    D:bool=NOT Boolean(A=Get bIsPlayingEmote())
   ) -> (ReturnValue)
   > Set CapeFEIdle(CapeFEIdle:double="To Float (Boolean)"(InBool=@n6)) -> (Output_Get:double)
   > Set CapeFEIdle_Bool(CapeFEIdle_Bool:bool=@n6) -> (Output_Get:bool)
   n7 = NOT Boolean(A=Get bIsPlayingEmote())
   n8 = NOT Boolean(A=Get bIsMoving2D())
   n9 = Get bIsInFrontEnd()
   > Set isWearing BackblingLongFEIdle(isWearing BackblingLongFEIdle:double="To Float (Boolean)"(InBool=AND Boolean(A=@n9, B=@n8, C=@n7, D=@n5.bIsBackblingLong))) -> (Output_Get:double)
   n10 = Get bIsSkyDiving()
   n11 = Get Is Wearing Cape()
   > Set CapeFESkydive(CapeFESkydive:double="To Float (Boolean)"(InBool=AND Boolean!(A=@n11, B=@n10) -> (ReturnValue))) -> (Output_Get:double)
   n12 = NOT Boolean(A=Get bIsSurfaceSwimming())
   n13 = NOT Boolean(A=Get bIsSkyDiving())
   n14 = NOT Boolean(A=Get bIsPlayingEmote())
   n15 = NOT Boolean(A=Get bIsSprinting())
   > Set HandTwistEnable(HandTwistEnable:double="To Float (Boolean)"(InBool=AND Boolean!(A=@n14, B=@n15, C:bool=@n13, D:bool=@n12) -> (ReturnValue))) -> (Output_Get:double)
   n16 = NOT Boolean(A=Get AdjustTailToEmotesBool() -> (AdjustTailToEmotesBool:bool))
   n17 = NOT Boolean(A=Get bIsInFrontEnd())
   n18 = Get Is Wearing Cape()
   > Set WearingCapeNoFE(WearingCapeNoFE:double="To Float (Boolean)"(InBool=AND Boolean!(A=@n18, B=@n17, C=@n16) -> (ReturnValue))) -> (Output_Get:double)
   n19 = Get bIsSliding()
   n20 = Get Is Wearing Cape()
   > Set SlidingWithCape(SlidingWithCape:double="To Float (Boolean)"(InBool=AND Boolean!(A=@n20, B=@n19) -> (ReturnValue))) -> (Output_Get:double)
   > n21 = Branch(Condition=Get HangGliding())
   > Set HangGlidingValue(HangGlidingValue:double=0) -> (Output_Get:double)
   Set HangGlidingValue(execute=@n21.else, HangGlidingValue=8)
   Event OnParachuteOpened()
   > Switch on EGliderType(Selection=Get PlayerGliderType() -> (PlayerGliderType=HangGlider))
   > Set HangGliding(HangGliding:bool=1) -> (Output_Get:bool)
   Event OnParachuteClosed()
   > Set HangGliding(HangGliding=false)
  ]}
]
variables [
 AdjustTailToEmotes {type=number}
 AdjustTailToEmotesBool {type=boolean}
 CapeFEIdle {type=number}
 CapeFEIdle_Bool {type=boolean}
 CapeFESkydive {type=number}
 Gravity Override {type=Vector}
 HandTwistEnable {type=number}
 HangGliding {type=boolean}
 HangGlidingValue {type=number, default=8}
 Is Wearing Cape {type=boolean}
 isWearing BackblingLongFEIdle {type=number}
 SlidingWithCape {type=number}
 WearingCapeNoFE {type=number}
]
