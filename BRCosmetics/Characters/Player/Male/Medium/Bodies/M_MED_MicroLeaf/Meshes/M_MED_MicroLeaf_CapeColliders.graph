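# Physics asset for the male medium MicroLeaf character body that, by its name, provides collision volumes for the character's cape.
# Capsules cover the clavicles, upper neck, five spine bones and four-bone tail chain, with a single sphere on the head.
# Arm, leg and foot bodies exist but hold no shapes.
# It defines no constraints, so it is a collider set rather than a jointed ragdoll.
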PhysicsAsset /BRCosmetics/Characters/Player/Male/Medium/Bodies/M_MED_MicroLeaf/Meshes/M_MED_MicroLeaf_CapeColliders
bodies {
 calf_l []
 calf_r []
 clavicle_l [{shapeType=Capsule, length=9.0192, radius=9.8871, center=<9.4266 -1.1806 -2.4977>, rotation=<rot 69.3297 1.4476 176.9462>}]
 clavicle_r [{shapeType=Capsule, length=9.0192, radius=9.8871, center=<-9.4266 1.1806 2.4977>, rotation=<rot 69.3297 1.4476 -3.0538>}]
 foot_l []
 foot_r []
 head [{shapeType=Sphere, radius=17.3125, center=<11.028 -0.174 0>}]
 lowerarm_l []
 lowerarm_r []
 neck_02 [{shapeType=Capsule, length=13.4375, radius=7.3175, center=<-0.2642 -0.3408 0>, rotation=<rot 90 -171.1193 -175.3978>}]
 spine_01 [{shapeType=Capsule, length=8, radius=29.3721, center=<-15.6956 -0.7488 -0.0001>}]
 spine_02 [
  {shapeType=Capsule, length=10.1619, radius=23.0145, center=<-1.711 -0.1851 0.0001>}
  {shapeType=Capsule, length=29.3956, radius=11.5807, center=<2.5024 0.4992 3.9825>, rotation=<rot 90 0.0049 0.0049>}
  {shapeType=Capsule, length=29.875, radius=11.1013, center=<2.4704 0.4991 -3.9825>, rotation=<rot 90 -0.0396 -0.0396>}
 ]
 spine_03 [{shapeType=Capsule, length=17.8366, radius=21.2208, center=<-8.9151 -0.898 0.0001>}]
 spine_04 [{shapeType=Capsule, length=4.6235, radius=20.4494, center=<2.766 0.7765 0.0002>}]
 spine_05 [
  {shapeType=Capsule, length=9.7057, radius=15.0119, center=<-2.2795 1.8109 0>}
  {shapeType=Capsule, length=13.9518, radius=13.9425, center=<-0.2916 0.4452 0>}
 ]
 tail1 [{shapeType=Capsule, length=31.8683, radius=16.209, center=<22.4551 2.6193 -0.1736>, rotation=<rot 90 0 0>}]
 tail2 [{shapeType=Capsule, length=28.3137, radius=13.3772, shapeName=tail2_capsule, center=<19.4974 0 0>, rotation=<rot 90 0 0>}]
 tail3 [{shapeType=Capsule, length=21.6244, radius=14.0269, shapeName=tail3_capsule, center=<13.5308 0 1.1061>, rotation=<rot 0 -90 -86.5076>}]
 tail4 [{shapeType=Capsule, length=20.8132, radius=10.8092, shapeName=tail4_capsule, center=<12.2475 0 5.0716>, rotation=<rot 0 90 -82.3256>}]
 thigh_l []
 thigh_r []
 upperarm_l []
 upperarm_r []
}
constraints []
